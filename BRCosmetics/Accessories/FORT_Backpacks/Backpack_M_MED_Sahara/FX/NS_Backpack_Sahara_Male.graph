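# Niagara effect for the male Sahara backpack cosmetic: cyan electric triangles with glows, a flickering light and smoke.
# Looping sprite emitters use spawn groups to trace triangle edges from three fixed points, while lightning meshes select positions and orientations from nine-point arrays.
# Glows and swirling smoke sample User.SkeletalMesh.
# Four anim-trail ribbons follow its bone or socket pairs, drifting upward as rising smoke and fading over each loop.
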NiagaraSystem /BRCosmetics/Accessories/FORT_Backpacks/Backpack_M_MED_Sahara/FX/NS_Backpack_Sahara_Male
dependencies {
 usedDataInterfaces [
  &NiagaraDataInterfaceCurve
  &NiagaraDataInterfaceColorCurve
  &NiagaraDataInterfaceVector2DCurve
  &NiagaraDataInterfaceArrayFloat3
 ]
 usedDynamicInputs []
 usedModules [&SystemState]
}
emitters {
 Air_Glows_and_Light {
  emitterUpdateScript [
   &EmitterState()
   &SpawnBurst_Instantaneous(Spawn Count=3, Use Spawn Probability=false, UseLoopCountLimit=true)
  ]
  particleSpawnScript [
   &FN_SetSpawnGroup(Number of Spawn Groups=3)
   &InitializeParticle(
    Color Maximum=<rgba 3.5333 8 6.9551 1>
    Color Minimum=<rgba 3.058 8 7.5217 1>
    Color Mode=::NewEnumerator2
    Sprite Size Mode=::NewEnumerator3
    Uniform Sprite Size=&MaskFloatBySpawnGroup(Masked Float=30, Passthrough Float=20)
    UsePositionOffset=false
    Write Material Random=true
   )
   &CameraOffset(Camera Offset Amount=10)
  ]
  particleUpdateScript [
   &ParticleState()
   &DynamicMaterialParameters(
    Index 0 Param 1=&FloatFromCurve(FloatCurve=DataInterface() {propertyValues {Curve {keys [
         <at 0 1 RCIM_Cubic>
         <at 0.5 1.5>
         {arriveTangent=3, interpMode=RCIM_Cubic, leaveTangent=3, time=0.8, value=1.5}
         {arriveTangent=4.6111, interpMode=RCIM_Cubic, leaveTangent=4.6111, tangentMode=RCTM_User, time=1, value=3}
        ]}}})
    Index 0 Param 2=8
    Index 0 Param 4=&RandomRangeFloat(Minimum=0, Maximum=1)
    Module.Param0WriteEnabled=true
    Module.Param1WriteEnabled=true
    Module.Param2WriteEnabled=true
    Module.Param3WriteEnabled=true
   )
   &ScaleColor(
    Scale Alpha=&FloatFromCurve(FloatCurve=DataInterface() {propertyValues {Curve {keys [
         <at 0.1 0.9 RCIM_Cubic>
         {arriveTangent=-1.125, interpMode=RCIM_Cubic, leaveTangent=-1.125, time=0.5, value=0.75}
         {interpMode=RCIM_Cubic, leaveTangent=-10.3917, time=0.9, value=0}
        ]}}})
    Scale RGB=&Multiply_VectorByFloat(
     Vector=&MakeVectorFromLinearColorRGB(Color=&ColorFromCurve(ColorCurve=DataInterface() {propertyValues {
         AlphaCurve {keys []}
         BlueCurve {keys [<at 0 9.6999>,<at 0.312 0.8>,<at 0.5 1.5>,<at 0.605 1.25>,<at 0.725 1.5>,<at 1 1>]}
         GreenCurve {keys [<at 0 10>,<at 0.312 0.7359>,<at 0.5 1.5>,<at 0.605 1.1373>,<at 0.725 1.2697>,<at 1 1>]}
         RedCurve {keys [<at 0 8.3043>,<at 0.312 0.4682>,<at 0.5 0.8125>,<at 0.605 0.6652>,<at 0.725 1.0234>,<at 1 1>]}
        }}))
     Float=1
    )
    ScaleA=true
    ScaleRGB=true
   )
   &SkeletalMeshLocation(
    Bone Sampling Mode=::NewEnumerator4
    MaskSpawnGroup=false
    Sampled Position Offset=&MaskVectorBySpawnGroup(
     Masked Vector=<0 -5 0>
     Passthrough Vector=&MaskVectorBySpawnGroup(Masked Vector=<10 -5 -5>, Passthrough Vector=<-10 -5 -5>, Spawn Group Mask=1)
    )
    Skeletal Mesh=$User.SkeletalMesh
   )
   &FN_Light_Attributes(
    Color=<rgba 0.3822 1 0.9402 1>
    Create Color Attribute=true
    Create Exponent Attribute=true
    Create Position Attribute=true
    Create Radius Attribute=true
    Enable Particle Light Enable=true
    Exponent=1
    Light Enabled:NiagaraBool=/Niagara/DynamicInputs/Helpers/MaskBoolBySpawnGroup(Masked Bool:NiagaraBool=true, Passthrough Bool:NiagaraBool=false, Spawn Group Mask:NiagaraInt32=0)
    Position=&Add_Vector(
     A=Linked() {linkedVariable {name=Engine.Owner.Position, type=NiagaraPosition}}
     B=&RotateVectorByQuaternion(Vector To Rotate=<0 -10 0>, Quaternion=$Engine.Owner.Rotation)
    )
    Radius=30
    Scale Color=&FloatFromCurve(
     FloatCurve=DataInterface() {propertyValues {Curve {keys [
         {arriveTangent=14.6548, interpMode=RCIM_Cubic, time=0, value=1}
         {arriveTangent=12.2799, interpMode=RCIM_Cubic, leaveTangent=12.2799, time=0.0604, value=0.1618}
         {arriveTangent=-0.0227, interpMode=RCIM_Cubic, leaveTangent=-0.0227, tangentMode=RCTM_User, time=0.1164, value=2.4295}
         {arriveTangent=-2.6888, interpMode=RCIM_Cubic, leaveTangent=-2.6888, time=0.1452, value=0.074}
         {arriveTangent=3.8271, interpMode=RCIM_Cubic, leaveTangent=3.8271, time=0.2784, value=1.9939}
         {arriveTangent=-8.3865, interpMode=RCIM_Cubic, leaveTangent=-8.3865, time=0.2924, value=0.6376}
         {arriveTangent=-3.3346, interpMode=RCIM_Cubic, leaveTangent=-3.3346, time=0.3724, value=1.2052}
         {arriveTangent=4.9466, interpMode=RCIM_Cubic, leaveTangent=4.9466, time=0.4274, value=0.1874}
         {arriveTangent=1.3103, interpMode=RCIM_Cubic, leaveTangent=1.3103, time=0.494, value=1.8066}
         {arriveTangent=0.5662, interpMode=RCIM_Cubic, leaveTangent=0.5662, time=0.6123, value=0.4296}
         {arriveTangent=-1.108, interpMode=RCIM_Cubic, leaveTangent=-1.108, time=0.7202, value=1.9347}
         <at 1 0 RCIM_Cubic>
        ]}}}
     Scale Curve=3
    )
   )
  ]
  rendererClasses [&NiagaraLightRendererProperties, &NiagaraSpriteRendererProperties]
  renderers [
   {rendererClass=&NiagaraLightRendererProperties, rendererIndex=0}
   {rendererClass=&NiagaraSpriteRendererProperties, rendererIndex=1}
  ]
 }
 LightningMeshes {
  emitterUpdateScript [
   &EmitterState(Life Cycle Mode=::NewEnumerator1, Loop Behavior=::NewEnumerator1, Loop Duration=0.5, UseLoopDelay=false)
   &SpawnRate(SpawnRate=12, Use Spawn Probability=false)
   &self:TrianglePositionArray(
    ArrayAmount=9
    Point 1:Vector3f=<4.5 -1 0>
    Point 2:Vector3f=<-4.5 -1 0>
    Point 3:Vector3f=<0 -1 8>
    Point 4:Vector3f=<15 -4 -8.75>
    Point 5:Vector3f=<9.25 -2 2>
    Point 6:Vector3f=<2.5 -4 -8.85>
    Point 7:Vector3f=<-3 -4 -9.35>
    Point 8:Vector3f=<0 -1 8>
    Point 9:Vector3f=<-15.5 -4 -8.85>
    Vector Array:NiagaraDataInterfaceArrayFloat3=$Emitter.Position Array
   ) {module=TrianglePositionArray}
   &self:TriangleOrientationArray(
    ArrayAmount=9
    Point 1:Vector3f=<4.5 -1 0>
    Point 2:Vector3f=<-4.5 -1 0>
    Point 3:Vector3f=<0 -1 8>
    Point 4:Vector3f=<15 -4 -8.75>
    Point 5:Vector3f=<9.25 -2 2>
    Point 6:Vector3f=<2.5 -4 -8.85>
    Point 7:Vector3f=<-3 -4 -9.35>
    Point 8:Vector3f=<-8.5 -2 3.15>
    Point 9:Vector3f=<-15.5 -4 -8.85>
    Vector Array:NiagaraDataInterfaceArrayFloat3=$Emitter.Vector Array
   ) {module=TriangleOrientationArray}
   /SetVariables(
    Emitter.Position Array=DataInterface() {propertyValues {GpuSyncMode=SyncCpuToGpu, MaxElements=0, FloatData [<0 0 0>, <0 0 0>, <0 0 0>, <0 0 0>, <0 0 0>, <0 0 0>, <0 0 0>, <0 0 0>, <0 0 0>]}}
    Emitter.Vector Array=DataInterface() {propertyValues {GpuSyncMode=SyncCpuToGpu, MaxElements=0, FloatData [<0 0 0>, <0 0 0>, <0 0 0>, <0 0 0>, <0 0 0>, <0 0 0>, <0 0 0>, <0 0 0>, <0 0 0>]}}
   ) {isSetParameters=true}
  ]
  particleSpawnScript [
   &PartitionParticles(Number Of Partitions=9, Partition Mode:ENiagara_PartitionMode=::NewEnumerator5)
   /SetVariables(Particles.Random Int:NiagaraInt32=$Particles.PartitionParticles.Partition) {isSetParameters=true}
   &InitializeParticle(
    Color=<rgba 1.0746 3 2.5785 1>
    Lifetime Max=0.35
    Lifetime Min=0.15
    Lifetime Mode=::NewEnumerator1
    Mesh Scale Max=<0.55 1.35 1.5>
    Mesh Scale Min=<0.45 1.1 1.35>
    Mesh Scale Mode=::NewEnumerator2
    Position Offset=&SelectVectorFromArray(Direct Array Index=$Particles.Random Int, Vector Selection Array=$Emitter.Position Array)
    Position Offset Coordinate Space=::Local
    Sprite Size Mode=::NewEnumerator3
    Uniform Sprite Size=1.5
    UsePositionOffset=true
    Write Material Random=true
   )
   &InitialMeshOrientation(
    Mesh Axis To Orient=&SelectVectorFromArray(Direct Array Index=$Particles.Random Int, Vector Selection Array=$Emitter.Vector Array)
    Mesh Orientation Mode=::NewEnumerator2
    Rotation=&RandomRangeVector(Minimum=<-1 0 0>, Maximum=<1 0 0>)
    Use Rotation Vector=true
   )
   &CameraOffset(Camera Offset Amount=&RandomRangeFloat(Minimum=0, Maximum=-2))
  ]
  particleUpdateScript [
   &ParticleState()
   &DynamicMaterialParameters(
    Index 0 Param 1=6
    Index 0 Param 2=&FloatFromCurve(FloatCurve=DataInterface() {propertyValues {Curve {keys [{arriveTangent=-2.9283, interpMode=RCIM_Cubic, leaveTangent=-2.9283, tangentMode=RCTM_User, time=0.1, value=1}, <at 1 -1>]}}})
    Index 0 Param 3=&RandomRangeFloat(Minimum=1.45, Maximum=2.1)
    Index 0 Param 4=&RandomRangeFloat(Minimum=-60, Maximum=60)
    Module.Param0WriteEnabled=true
    Module.Param1WriteEnabled=true
    Module.Param2WriteEnabled=true
    Module.Param3WriteEnabled=true
   )
   &ScaleColor(
    Scale Alpha=&FloatFromCurve(FloatCurve=DataInterface() {propertyValues {Curve {keys [
         {arriveTangent=9.0725, interpMode=RCIM_Cubic, leaveTangent=9.0725, tangentMode=RCTM_User, time=0, value=0}
         {arriveTangent=-2.5249, interpMode=RCIM_Cubic, leaveTangent=-2.5249, tangentMode=RCTM_User, time=0.1408, value=0.9847}
         {arriveTangent=-0.0405, interpMode=RCIM_Cubic, leaveTangent=-0.0405, tangentMode=RCTM_User, time=1, value=0}
        ]}}})
    Scale RGB=&MakeVectorFromLinearColorRGB(Color=&ColorFromCurve(ColorCurve=DataInterface() {propertyValues {
        AlphaCurve {keys [<at 0 1 RCIM_Cubic>]}
        BlueCurve {keys [<at 0 10 RCIM_Cubic>,<at 0.853 1 RCIM_Cubic>]}
        GreenCurve {keys [<at 0 10 RCIM_Cubic>,<at 0.853 1 RCIM_Cubic>]}
        RedCurve {keys [<at 0 10 RCIM_Cubic>,<at 0.853 1 RCIM_Cubic>]}
       }}))
    ScaleA=true
    ScaleRGB=true
   )
  ]
  rendererClasses [&NiagaraMeshRendererProperties]
  renderers [{rendererClass=&NiagaraMeshRendererProperties, rendererIndex=0}]
 }
 Main_Triangle_Electric_Sprites {
  emitterUpdateScript [
   &EmitterState(Life Cycle Mode=::NewEnumerator1, Loop Behavior=::NewEnumerator1, Loop Duration=0.5, UseLoopDelay=false)
   &SpawnRate(SpawnRate=18, Use Spawn Probability=false)
  ]
  particleSpawnScript [
   &FN_SetSpawnGroup(Number of Spawn Groups=3)
   &InitializeParticle(
    Color=<rgba 5.4119 9.801 10 0.25>
    Lifetime Max=0.3
    Lifetime Min=0.15
    Lifetime Mode=::NewEnumerator1
    Sprite Size Max=<14 16>
    Sprite Size Min=<12 16>
    Sprite Size Mode=::NewEnumerator2
    UsePositionOffset=false
    Write Material Random=true
   )
   &DynamicMaterialParameters(
    Index 0 Param 1=&UniformRangedFloat(Minimum=2.5, Maximum=3.5)
    Index 0 Param 3=&UniformRangedFloat(Minimum=0.2, Maximum=0.5)
    Module.Param0WriteEnabled=true
    Module.Param0WriteEnabled2=true
    Module.Param1WriteEnabled=false
    Module.Param1WriteEnabled2=true
    Module.Param2WriteEnabled=true
    Module.Param2WriteEnabled2=true
    Module.Param3WriteEnabled=true
    Module.Param3WriteEnabled2=true
    Write Parameter Index 2=true
   )
   &CameraOffset(Camera Offset Amount=-2)
   /SetVariables(Particles.LightningUVScale:Vector2f=&MakeVector2D(X=&UniformAOrBFloat(A=-1), Y=&UniformAOrBFloat(A=-1))) {isSetParameters=true, module=SetVariables_#1}
   /SetVariables(Particles.Point_One:Vector3f=<8 -2.5 -6>, Particles.Point_Three:Vector3f=<0 -2.5 8>, Particles.Point_Two:Vector3f=<-8 -2.5 -6>) {isSetParameters=true, module=SetVariables_#2}
   &SystemLocation(
    MaskSpawnGroup=false
    Offset=&MaskVectorBySpawnGroup(
     Masked Vector=$Particles.Point_One
     Passthrough Vector=&MaskVectorBySpawnGroup(Masked Vector=$Particles.Point_Two, Passthrough Vector=$Particles.Point_Three, Spawn Group Mask=1)
    )
   )
   &SpriteFacingAndAlignment(
    Sprite Alignment=&MaskVectorBySpawnGroup(
     Masked Vector=&Subtract_Vector(A=$Particles.Point_One, B=$Particles.Point_Three)
     Passthrough Vector=&MaskVectorBySpawnGroup(
      Masked Vector=&Subtract_Vector(A=$Particles.Point_Two, B=$Particles.Point_One)
      Passthrough Vector=&Subtract_Vector(A=$Particles.Point_Three, B=$Particles.Point_Two)
      Spawn Group Mask=1
     )
    )
    WriteAlignment=true
    WriteFacing=false
   )
  ]
  particleUpdateScript [
   &ParticleState()
   &ScaleSpriteSize(
    Non-Uniform Curve Sprite Scale=DataInterface() {propertyValues {
      XCurve {keys [
        {arriveTangent=2.2276, interpMode=RCIM_Cubic, leaveTangent=2.2276, tangentMode=RCTM_User, time=0, value=0.3}
        {tangentMode=RCTM_User, time=1, value=1}
       ]}
      YCurve {keys [{tangentMode=RCTM_User, time=0, value=1}, {tangentMode=RCTM_User, time=1, value=1}]}
     }}
    Scale Sprite Size Mode=::NewEnumerator3
   )
  ]
  rendererClasses [&NiagaraSpriteRendererProperties]
  renderers [{rendererClass=&NiagaraSpriteRendererProperties}]
 }
 Rising_Smoke_Ribbon {
  emitterUpdateScript [
   &EmitterState(Inactive Response=::NewEnumerator1, Life Cycle Mode=::NewEnumerator1, Loop Behavior=::NewEnumerator1, Loop Duration=3, UseLoopDelay=false)
   &FN_SpawnAnimTrail(Number of Anim Trails=4, Spawn Interval=0)
  ]
  particleSpawnScript [
   &FN_InitializeAnimTrailBoneIndecies()
   &FN_InitializeAnimTrailBoneIndecies(AnimTrail Selection=1, First Bone or Socket Index=3, Second Bone or Socket Index=2) {module=FN_InitializeAnimTrailBoneIndecies001}
   &FN_InitializeAnimTrailBoneIndecies(AnimTrail Selection=2, First Bone or Socket Index=4, Second Bone or Socket Index=5) {module=FN_InitializeAnimTrailBoneIndecies002}
   &FN_InitializeAnimTrailBoneIndecies(AnimTrail Selection=3, First Bone or Socket Index=6, Second Bone or Socket Index=7) {module=FN_InitializeAnimTrailBoneIndecies003}
   &FN_InitializeAnimTrail(Lifetime=&RandomRangeFloat(Minimum=0.8, Maximum=1.2, Randomness Mode=::NewEnumerator1, Random Seed=$Particles.SpawnGroup, Fixed Random Seed=true))
   &AddVelocity(Velocity=&RandomRangeVector(Minimum=<0 0 35>, Maximum=<0 0 80>, Randomness Mode=::NewEnumerator1, Random Seed=$Particles.SpawnGroup, Fixed Random Seed=true))
  ]
  particleUpdateScript [
   &ParticleState()
   &FN_UpdateAnimTrail(SkeletalMesh=$User.SkeletalMesh)
   &InheritSourceMovement(
    Applied Position Scale=&VectorFromFloat(Value=&FloatFromCurve(FloatCurve=DataInterface() {propertyValues {Curve {keys [
          {tangentMode=RCTM_User, time=0, value=1}
          {arriveTangent=-3.9721, interpMode=RCIM_Cubic, leaveTangent=-3.9721, tangentMode=RCTM_User, time=0.7, value=0.7}
         ]}}}))
    ApplyForce=false
    ApplyPosition=true
    ApplyPositionFalloff=false
   )
   &Drag(Use Linear Drag=true, Use Rotational Drag=true)
   &SolveForcesAndVelocity(Clamp Velocity=false, Limit Acceleration=false)
   &FN_RibbonMaterial(
    Color=&ColorFromCurve(ColorCurve=DataInterface() {propertyValues {
       AlphaCurve {keys [<at 0 0.3 RCIM_Cubic>]}
       BlueCurve {keys [<at 0.1126 0.3957 RCIM_Cubic>,<at 0.7758 0.5365>]}
       GreenCurve {keys [<at 0.1126 0.4167 RCIM_Cubic>,<at 0.7758 0.5365>]}
       RedCurve {keys [<at 0.1126 0.3272 RCIM_Cubic>,<at 0.7758 0.5365>]}
      }})
    Color Scale U=0
    Erosion Power=0
    Erossion Scale U=0
    Macro Scale U=0
    Micro Scale U=0
    Opacity Pan U Speed:NiagaraFloat=&RandomRangeFloat(Minimum=0.35, Maximum=0.6, Randomness Mode=::NewEnumerator1, Random Seed=$Particles.SpawnGroup, Fixed Random Seed=true)
    Opacity Power=3
    Opacity Scale U:NiagaraFloat=&RandomRangeFloat(Minimum=0.6, Maximum=1.2, Randomness Mode=::NewEnumerator1, Random Seed=$Particles.SpawnGroup, Fixed Random Seed=true)
    Scale Alpha:NiagaraFloat=&FloatFromCurve(
     FloatCurve=DataInterface() {propertyValues {Curve {keys [
         {arriveTangent=14.1317, interpMode=RCIM_Cubic, leaveTangent=14.1317, tangentMode=RCTM_User, time=0, value=0}
         <at 0.1 1>
         {arriveTangent=-3.2156, interpMode=RCIM_Cubic, leaveTangent=-3.2156, tangentMode=RCTM_User, time=0.5, value=0.95}
         {interpMode=RCIM_Cubic, tangentMode=RCTM_User, time=1, value=0}
        ]}}}
     Scale Curve=&FloatFromCurve(
      FloatCurve=DataInterface() {propertyValues {Curve {keys [
          <at 0 0 RCIM_Cubic>
          {arriveTangent=10.8383, interpMode=RCIM_Cubic, tangentMode=RCTM_Break, time=0.3, value=1}
          {arriveTangent=-0.263, interpMode=RCIM_Cubic, leaveTangent=-0.263, tangentMode=RCTM_User, time=0.6, value=1}
          <at 1 0 RCIM_Cubic>
         ]}}}
      CurveIndex=$Emitter.NormalizedLoopAge
     )
    )
    Write to Module Namespace=true
   )
   &ScaleRibbonWidth(Ribbon Width Scale=&FloatFromCurve(FloatCurve=DataInterface() {propertyValues {Curve {keys [{tangentMode=RCTM_User, time=0, value=0}, {tangentMode=RCTM_User, time=0.5, value=1}, <at 1 0>]}}}))
   /SetVariables(Particles.RibbonLinkOrder:NiagaraFloat=$Particles.NormalizedAge) {isSetParameters=true}
  ]
  rendererClasses [&NiagaraRibbonRendererProperties]
  renderers [{rendererClass=&NiagaraRibbonRendererProperties}]
 }
 Secondary_Triangles_Electric_Sprites {
  emitterUpdateScript [
   &EmitterState(Life Cycle Mode=::NewEnumerator1, Loop Behavior=::NewEnumerator1, Loop Duration=0.5, UseLoopDelay=false)
   &SpawnRate(SpawnRate=26, Use Spawn Probability=false)
  ]
  particleSpawnScript [
   &FN_SetSpawnGroup(Number of Spawn Groups=3)
   &InitializeParticle(
    Color=<rgba 5.4119 9.801 10 0.25>
    Lifetime Max=0.3
    Lifetime Min=0.15
    Lifetime Mode=::NewEnumerator1
    Sprite Size Max=<10 16>
    Sprite Size Min=<8 12>
    Sprite Size Mode=::NewEnumerator2
    UsePositionOffset=false
    Write Material Random=true
   )
   &DynamicMaterialParameters(
    Index 0 Param 1=&UniformRangedFloat(Minimum=2.5, Maximum=3.5)
    Index 0 Param 3=&UniformRangedFloat(Minimum=0.2, Maximum=0.5)
    Module.Param0WriteEnabled=true
    Module.Param0WriteEnabled2=true
    Module.Param1WriteEnabled=false
    Module.Param1WriteEnabled2=true
    Module.Param2WriteEnabled=true
    Module.Param2WriteEnabled2=true
    Module.Param3WriteEnabled=true
    Module.Param3WriteEnabled2=true
    Write Parameter Index 2=true
   )
   &CameraOffset(Camera Offset Amount=-2)
   /SetVariables(Particles.LightningUVScale:Vector2f=&MakeVector2D(X=&UniformAOrBFloat(A=-1), Y=&UniformAOrBFloat(A=-1))) {isSetParameters=true, module=SetVariables_#1}
   /SetVariables(Particles.Point_One:Vector3f=<8 -2.5 -6.5>, Particles.Point_Three:Vector3f=<2.5 -2.5 6>, Particles.Point_Two:Vector3f=<-4.5 -2.5 -6>) {isSetParameters=true, module=SetVariables_#2}
   &SystemLocation(
    MaskSpawnGroup=false
    Offset=&Add_Vector(
     A=&MaskVectorBySpawnGroup(
      Masked Vector=$Particles.Point_One
      Passthrough Vector=&MaskVectorBySpawnGroup(Masked Vector=$Particles.Point_Two, Passthrough Vector=$Particles.Point_Three, Spawn Group Mask=1)
     )
     B=&Lerp_Vector(A=<7 0 -2.85>, B=<-11 0 -2.85>, Alpha=&UniformAOrBFloat(A=0))
    )
   )
   &SpriteFacingAndAlignment(
    Sprite Alignment=&MaskVectorBySpawnGroup(
     Masked Vector=&Subtract_Vector(A=$Particles.Point_One, B=$Particles.Point_Three)
     Passthrough Vector=&MaskVectorBySpawnGroup(
      Masked Vector=&Subtract_Vector(A=$Particles.Point_Two, B=$Particles.Point_One)
      Passthrough Vector=&Subtract_Vector(A=$Particles.Point_Three, B=$Particles.Point_Two)
      Spawn Group Mask=1
     )
    )
    WriteAlignment=true
    WriteFacing=false
   )
  ]
  particleUpdateScript [
   &ParticleState()
   &ScaleSpriteSize(
    Non-Uniform Curve Sprite Scale=DataInterface() {propertyValues {
      XCurve {keys [
        {arriveTangent=2.2276, interpMode=RCIM_Cubic, leaveTangent=2.2276, tangentMode=RCTM_User, time=0, value=0.3}
        {tangentMode=RCTM_User, time=1, value=1}
       ]}
      YCurve {keys [{tangentMode=RCTM_User, time=0, value=1}, {tangentMode=RCTM_User, time=1, value=1}]}
     }}
    Scale Sprite Size Mode=::NewEnumerator3
   )
  ]
  rendererClasses [&NiagaraSpriteRendererProperties]
  renderers [{rendererClass=&NiagaraSpriteRendererProperties}]
 }
 Swirly_Smoke {
  emitterUpdateScript [
   &EmitterState(Life Cycle Mode=::NewEnumerator1, Loop Behavior=::NewEnumerator1, Loop Delay=0.25, Loop Duration=1.25, UseLoopDelay=true)
   &SpawnBurst_Instantaneous(Spawn Count=10, Use Spawn Probability=false, UseLoopCountLimit=false)
   &SpawnRate(SpawnRate=30, Use Spawn Probability=false)
  ]
  particleSpawnScript [
   &PartitionParticles(Number Of Partitions=10)
   &InitializeParticle(
    Color=<rgba 0.4826 0.6146 0.5837 0.2>
    Lifetime Max=1.1
    Lifetime Mode=::NewEnumerator1
    Sprite Rotation Mode=::NewEnumerator2
    Sprite Size Mode=::NewEnumerator1
    Uniform Sprite Size Max=18
    Uniform Sprite Size Min=15
    UsePositionOffset=false
    Write Material Random=true
   )
   &DynamicMaterialParameters(
    Index 0 Param 1=&RandomRangeFloat(Minimum=0, Maximum=1, Evaluation Type=::NewEnumerator1)
    Index 0 Param 2=&RandomRangeFloat(Minimum=2, Maximum=4, Evaluation Type=::NewEnumerator1)
    Module.Param0WriteEnabled=true
    Module.Param1WriteEnabled=true
    Module.Param2WriteEnabled=false
    Module.Param3WriteEnabled=false
   )
   &CameraOffset()
  ]
  particleUpdateScript [
   &ParticleState()
   &SkeletalMeshLocation(
    Bone / Socket Index=$Particles.PartitionParticles.Partition
    MaskSpawnGroup=false
    Mesh Sampling Type=::NewEnumerator5
    Skeletal Mesh=$User.SkeletalMesh
    Socket Sampling Mode=::NewEnumerator5
   )
   &ScaleColor(
    Scale Alpha=&FloatFromCurve(FloatCurve=DataInterface() {propertyValues {Curve {keys [
         {arriveTangent=9.0725, interpMode=RCIM_Cubic, leaveTangent=9.0725, tangentMode=RCTM_User, time=0, value=0}
         {arriveTangent=-2.5249, interpMode=RCIM_Cubic, leaveTangent=-2.5249, tangentMode=RCTM_User, time=0.1408, value=0.9847}
         {arriveTangent=-0.0405, interpMode=RCIM_Cubic, leaveTangent=-0.0405, tangentMode=RCTM_User, time=0.9, value=0}
        ]}}})
    ScaleA=true
    ScaleRGB=true
   )
   &SubUVAnimation(
    Start Frame:NiagaraInt32=&RandomRangeInt(Minimum=0, Maximum=30)
    Sub UV Play Rate=0.1
    SubUV Animation Mode:ENiagara_SubUVLookupMode=::NewEnumerator3
   )
   &ScaleSpriteSize(Scale Factor=&Vector2DFromFloat(Value=&FloatFromCurve(FloatCurve=DataInterface() {propertyValues {Curve {keys [
          {arriveTangent=2.7647, interpMode=RCIM_Cubic, leaveTangent=2.7647, tangentMode=RCTM_User, time=0, value=0.2}
          {interpMode=RCIM_Cubic, leaveTangent=0.3663, time=1.0038, value=1.5}
         ]}}})))
  ]
  rendererClasses [&NiagaraSpriteRendererProperties]
  renderers [{rendererClass=&NiagaraSpriteRendererProperties}]
 }
}
summary {
 systemName=NS_Backpack_Sahara_Male
 emitters [
  {
   emitterName=Air_Glows_and_Light
   rendererClasses [&NiagaraLightRendererProperties, &NiagaraSpriteRendererProperties]
  }
  {emitterName=Main_Triangle_Electric_Sprites, rendererClasses [&NiagaraSpriteRendererProperties]}
  {emitterName=Secondary_Triangles_Electric_Sprites, rendererClasses [&NiagaraSpriteRendererProperties]}
  {emitterName=LightningMeshes, rendererClasses [&NiagaraMeshRendererProperties]}
  {emitterName=Swirly_Smoke, rendererClasses [&NiagaraSpriteRendererProperties]}
  {emitterName=Rising_Smoke_Ribbon, rendererClasses [&NiagaraRibbonRendererProperties]}
 ]
 userVariables [
  User.RibbonIDOffset {type=NiagaraInt32}
  User.SkeletalMesh {
   type=NiagaraDataInterfaceSkeletalMesh
   defaultValue {
    struct=NiagaraExt_VariableValue_DataInterface
    value {dataInterface=&self:NiagaraDataInterfaceSkeletalMesh_1}
   }
  }
 ]
}
