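# Animation Blueprint for the Sahara backpack cosmetic, which shakes in response to player events.
# Anim input events for chest opened, enemy eliminated, or shield gain holding SlurpJuice set ShakingAnimation, cleared after each evaluation.
# That pulse moves a state machine from empty Idle into a one-shot shake, applied additively over the reference pose.
# A Modify Bone node, gated by bIsInFrontEnd, offsets the pack bone; the bound loot delegate has no logic.
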AnimBlueprint /BRCosmetics/Accessories/FORT_Backpacks/Backpack_M_MED_Sahara/Meshes/M_MED_Sahara_Pack_AnimBP
target_skeleton=&/M_MED_Sahara_Pack_Skeleton
graphs [
 AnimGraph {nodes [
   n1 = Local Space Ref Pose()
   AnimGraphNode_StateMachine_0 = Shaking_Idle() {is=title AnimGraphNode_StateMachine}
   n2 = "Transform (Modify) Bone - Bone: pack"!(
    ComponentPose=Local To Component(LocalPose=Apply Additive(Base=@n1, Additive=@AnimGraphNode_StateMachine_0, Alpha=0.95) {alphaBoolBlend {blendOption=Linear}})
    bAlphaBoolEnabled:bool=Get bIsInFrontEnd()
    Translation="0,-2.5,0"
    Rotation="0,0,0"
    Scale
   ) -> (Pose) {
    alphaInputType=Bool
    rotationMode=BMM_Additive
    translationMode=BMM_Additive
    alphaBoolBlend {blendOption=Linear}
    boneToModify {boneName=pack}
    translation {y=-2.5}
   }
   Output Pose(Result=Component To Local(ComponentPose=@n2)) {layerGroup=None}
  ]}
 Shaking_Idle {
  entry=Idle
  path=AnimGraph.AnimGraphNode_StateMachine_0.Shaking_Idle
  ids=id
  nodes [
   AnimStateNode_0 = Idle()
   AnimStateNode_1 = Shaking() {is=title AnimStateNode, bAlwaysResetOnEntry=true}
   AnimStateTransitionNode_0 = AnimStateTransitionNode()
   AnimStateTransitionNode_1 = AnimStateTransitionNode()
  ]
  links [
   {
    dst=Shaking
    rule=AnimGraph.AnimGraphNode_StateMachine_0.Shaking_Idle.AnimStateTransitionNode_0.Transition
    src=Idle
    via=AnimStateTransitionNode_0
   }
   {
    dst=Idle
    rule=AnimGraph.AnimGraphNode_StateMachine_0.Shaking_Idle.AnimStateTransitionNode_1.Transition
    src=Shaking
    via=AnimStateTransitionNode_1
   }
  ]
 }
 Idle {path=AnimGraph.AnimGraphNode_StateMachine_0.Shaking_Idle.AnimStateNode_0.Idle, nodes [Output Animation Pose()]}
 Shaking {
  path=AnimGraph.AnimGraphNode_StateMachine_0.Shaking_Idle.AnimStateNode_1.Shaking
  nodes [Output Animation Pose(Result="Sequence Player (additive) 'Sahara_Shaking_Backpack'"() {bLoopAnimation=false, sequence=/BRCosmetics/Accessories/FORT_Backpacks/Backpack_M_MED_Sahara/Animation/Sahara_Shaking_Backpack})]
 }
 Transition {
  path=AnimGraph.AnimGraphNode_StateMachine_0.Shaking_Idle.AnimStateTransitionNode_0.Transition
  nodes [Result(bCanEnterTransition=Get ShakingAnimation() -> (ShakingAnimation:bool))]
 }
 Transition {
  path=AnimGraph.AnimGraphNode_StateMachine_0.Shaking_Idle.AnimStateTransitionNode_1.Transition
  nodes [Result(bCanEnterTransition="float <= float"(A="Time Remaining (Sahara_Shaking_Backpack)"!() -> (ReturnValue:float) {is=title K2Node_AnimGetter}, B=0.1) -> (ReturnValue=~))]
 }
 EventGraph {nodes [
   Event OnCharacterCustomizationLoaded()
   > n1 = Cast To PlayerPawn_Athena(Object=TryGetPawnOwner())
   > n2 = Set As Player Pawn Athena(As Player Pawn Athena:Player Pawn Athena Object Reference=@n1) -> (Output_Get:Player Pawn Athena Object Reference)
   > Set BodyMesh(BodyMesh=GetSkeletalMeshForPartType(self=@n1))
   > n3 = FindGender!(execute, Object:Object Reference=@n2) -> (True:Exec, False:Exec, OutGender:EFortCustomGender Enum=Invalid)
   > n4 = Set isFemale(isFemale:bool="Equal (Enum)"(A:EFortCustomGender Enum=@n3, B:EFortCustomGender Enum=Female)) -> (Output_Get:bool)
   Bind Event to On Player Looted Container(
    execute=[@n4.then, @n3.False]
    self:Fort Player Pawn Athena Object Reference=@n1
    Delegate=Loot!() -> (OutputDelegate, then, FortPlayerPawnAthena:Fort Player Pawn Athena Object Reference, LootedContainer:Building Container Object Reference) {is=title K2Node_CustomEvent}
   ) {is=title K2Node_AddDelegate}
   n5 = Event BlueprintProcessAnimInputEvent()
   n6 = IsNamed(self=@n5.AnimInputEvent, InEventName=aie.PlayerVitality.ShieldsIncreased)
   n7 = Get As Player Pawn Athena() -> (As Player Pawn Athena:Player Pawn Athena Object Reference)
   n8 = Contains(
    SearchIn=GetDisplayName(Object=Get CurrentWeapon!(self:Fort Pawn Object Reference=@n7) -> (CurrentWeapon:Fort Weapon Object Reference))
    Substring=SlurpJuice
   )
   > Branch(Condition=OR Boolean!(
     A=IsNamed(self=@n5.AnimInputEvent, InEventName=aie.ChestOpened)
     B=IsNamed(self=@n5.AnimInputEvent, InEventName=aie.EnemyPlayerEliminated)
     C:bool=AND Boolean(A=@n6, B=@n8)
    ) -> (ReturnValue))
   > Set ShakingAnimation(ShakingAnimation:bool=true) -> (Output_Get:bool)
   Event BlueprintPostEvaluateAnimation() {is=title K2Node_Event}
   > Set ShakingAnimation(ShakingAnimation=false)
   TryGetPawnOwner()
   Event BlueprintUpdateAnimation()
  ]}
]
variables [
 As Player Pawn Athena {type=PlayerPawn_Athena_C}
 BodyMesh {type=SkeletalMeshComponent}
 isFemale {type=boolean}
 ShakingAnimation {type=boolean}
 shouldMovePack {type=boolean}
]
